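# Sound cue for the start of the crack lightning end event, playing the Cube_EndEvent_Start_01 wave.
# A single wave player feeds the output directly, with no randomization or mixing.
# It uses the shared Crack_Lightning_Att attenuation and the important in-game mission sound class.
# A concurrency override allows one instance at a time, stopping the oldest when it retriggers.
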SoundCue /Figment_S05/Athena/Sounds/Events/CrackLightning/Crack_EndEvent_Start_Cue
root=WavePlayer_0
nodes [{id=WavePlayer_0, properties {soundWaveAssetPtr=&/Cube_EndEvent_Start_01}}]
properties {
 attenuationSettings=/FNGameplayCues/Athena/Sounds/Events/CrackLightning/Crack_Lightning_Att
 bOverrideConcurrency=true
 maxDistance=2097152
 soundClassObject=&Fort_Mission_Ingame_Important_PSM
 volumeMultiplier=1
 attenuationOverrides {
  bAttenuate=false
  bEnableFocusInterpolation=true
  bEnableListenerFocus=true
  bEnableOcclusion=true
  focusAttackInterpSpeed=0.1
  focusAzimuth=15
  focusReleaseInterpSpeed=0.5
  nonFocusAzimuth=30
  nonFocusVolumeAttenuation=0
  occlusionInterpolationTime=2
  occlusionTraceChannel=ECC_GameTraceChannel11
  occlusionVolumeAttenuation=0
  stereoSpread=30000
 }
 concurrencyOverrides {maxCount=1, resolutionRule=StopOldest}
}
