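# Sound cue for the looping lightning beam of the Crack Lightning event's final strike.
# A looping beam wave plays after a short fixed delay, inside an indefinitely looping Enveloper that holds volume flat and ramps pitch upward, climbing sharply at the end of its curve.
# It uses the shared Crack Lightning attenuation asset and an important in-game mission sound class.
# Overridden concurrency allows a single instance, stopping the oldest.
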SoundCue /Figment_S05/Athena/Sounds/Events/CrackLightning/Crack_Lightning_Strike_Final_Loop_Cue
root=Enveloper_0
nodes [
 {id=Delay_0, children [WavePlayer_0], properties {delayMax=1, delayMin=1}}
 {
  id=Enveloper_0
  children [Delay_0]
  properties {
   bLoop=true
   bLoopIndefinitely=true
   pitchCurve {editorCurveData {keys [<at 0 0.8>,<at 13 1.1>,<at 15 1.5>]}}
   volumeCurve {editorCurveData {keys [<at 0 1>]}}
  }
 }
 {id=WavePlayer_0, properties {bLooping=true, soundWaveAssetPtr=/Game/Athena/Sounds/Events/CrackLightning/Crack_Lightning_BeamLoop_01}}
]
properties {
 attenuationSettings=/FNGameplayCues/Athena/Sounds/Events/CrackLightning/Crack_Lightning_Att
 bOverrideConcurrency=true
 maxDistance=2097152
 soundClassObject=&Fort_Mission_Ingame_Important_PSM
 volumeMultiplier=1
 attenuationOverrides {
  bAttenuate=false
  bEnableFocusInterpolation=true
  bEnableListenerFocus=true
  bEnableOcclusion=true
  focusAttackInterpSpeed=0.1
  focusAzimuth=15
  focusReleaseInterpSpeed=0.5
  nonFocusAzimuth=30
  nonFocusVolumeAttenuation=0
  occlusionInterpolationTime=2
  occlusionTraceChannel=ECC_GameTraceChannel11
  occlusionVolumeAttenuation=0
  stereoSpread=30000
 }
 concurrencyOverrides {maxCount=1, resolutionRule=StopOldest}
}
