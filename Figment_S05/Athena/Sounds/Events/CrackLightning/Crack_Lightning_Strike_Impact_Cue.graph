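# Sound cue for a Crack Lightning strike impact, blending close and distant layers of the hit.
# Each layer randomly picks one of two waves and passes it through its own close or distant attenuation asset.
# A mixer plays both layers together at equal volume into a root modulator.
# A concurrency override allows one instance at a time, stopping the oldest, and the cue uses an important in-game mission sound class.
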SoundCue /Figment_S05/Athena/Sounds/Events/CrackLightning/Crack_Lightning_Strike_Impact_Cue
root=Modulator_2
nodes [
 {
  id=Attenuation_0
  children [Random_1]
  properties {attenuationSettings=&/Crack_Lightning_Impact_Close_Att}
 }
 {
  id=Attenuation_1
  children [Random_2]
  properties {attenuationSettings=&/Crack_Lightning_Impact_Distant_Att}
 }
 {id=Mixer_0, children [Attenuation_0, Attenuation_1], properties {inputVolume [1,1]}}
 {id=Modulator_2, children [Mixer_0]}
 {id=Random_1, children [WavePlayer_0,WavePlayer_1], properties {weights [1,1]}}
 {id=Random_2, children [WavePlayer_2,WavePlayer_3], properties {weights [1,1]}}
 {id=WavePlayer_0, properties {soundWaveAssetPtr=/Game/Athena/Sounds/Events/CrackLightning/Crack_Lightning_StrikeImpact_01}}
 {id=WavePlayer_1, properties {soundWaveAssetPtr=/Game/Athena/Sounds/Events/CrackLightning/Crack_Lightning_StrikeImpact_02}}
 {id=WavePlayer_2, properties {soundWaveAssetPtr=&/Crack_Lightning_StrikeImpactDistant_01}}
 {id=WavePlayer_3, properties {soundWaveAssetPtr=&/Crack_Lightning_StrikeImpactDistant_02}}
]
properties {
 bOverrideConcurrency=true
 maxDistance=60000
 soundClassObject=&Fort_Mission_Ingame_Important_PSM
 volumeMultiplier=1
 attenuationOverrides {
  bAttenuate=false
  bEnableFocusInterpolation=true
  bEnableListenerFocus=true
  bEnableOcclusion=true
  focusAttackInterpSpeed=0.1
  focusAzimuth=15
  focusReleaseInterpSpeed=0.5
  nonFocusAzimuth=30
  nonFocusVolumeAttenuation=0
  occlusionInterpolationTime=2
  occlusionTraceChannel=ECC_GameTraceChannel11
  occlusionVolumeAttenuation=0
  stereoSpread=30000
 }
 concurrencyOverrides {maxCount=1, resolutionRule=StopOldest}
}
