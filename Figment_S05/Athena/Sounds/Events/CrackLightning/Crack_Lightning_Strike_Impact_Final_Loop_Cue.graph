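# Looping sound cue for the final Crack Lightning strike beam impact, limited to one playing instance with stop-oldest resolution.
# Close and distant layers each have their own attenuation and are mixed equally into a root modulator.
# Each layer picks full or optimized looping waves by quality level.
# Looping envelopers hold volume flat and key pitch to climb from below to above normal over a 30-second curve.
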SoundCue /Figment_S05/Athena/Sounds/Events/CrackLightning/Crack_Lightning_Strike_Impact_Final_Loop_Cue
root=Modulator_0
nodes [
 {
  id=Attenuation_2
  children [Enveloper_0]
  properties {attenuationSettings=&/Crack_Lightning_Impact_Close_Att}
 }
 {
  id=Attenuation_3
  children [Enveloper_1]
  properties {attenuationSettings=&/Crack_Lightning_Impact_Distant_Att}
 }
 {
  id=Enveloper_0
  children [QualityLevel_0]
  properties {
   bLoop=true
   bLoopIndefinitely=true
   pitchCurve {editorCurveData {keys [<at 0 0.8>,<at 30 1.25>]}}
   volumeCurve {editorCurveData {keys [<at 0 1>]}}
  }
 }
 {
  id=Enveloper_1
  children [QualityLevel_1]
  properties {
   bLoop=true
   bLoopIndefinitely=true
   pitchCurve {editorCurveData {keys [<at 0 0.8>,<at 30 1.25>]}}
   volumeCurve {editorCurveData {keys [<at 0 1>]}}
  }
 }
 {id=Mixer_2, children [Attenuation_2, Attenuation_3], properties {inputVolume [1,1]}}
 {id=Modulator_0, children [Mixer_2]}
 {id=QualityLevel_0, children [WavePlayer_4,WavePlayer_4,WavePlayer_1,WavePlayer_1]}
 {id=QualityLevel_1, children [WavePlayer_5,WavePlayer_5,WavePlayer_0,WavePlayer_0]}
 {
  id=WavePlayer_0
  properties {bLooping=true, soundWaveAssetPtr=&/Crack_Lightning_BeamLoop_ImpactDistant_01_opt}
 }
 {
  id=WavePlayer_1
  properties {bLooping=true, soundWaveAssetPtr=&/Crack_Lightning_BeamLoop_Impact_01_opt}
 }
 {id=WavePlayer_4, properties {bLooping=true, soundWaveAssetPtr=/Game/Athena/Sounds/Events/CrackLightning/Crack_Lightning_BeamLoop_Impact_01}}
 {
  id=WavePlayer_5
  properties {bLooping=true, soundWaveAssetPtr=/Game/Athena/Sounds/Events/CrackLightning/Crack_Lightning_BeamLoop_ImpactDistant_01}
 }
]
properties {
 bOverrideConcurrency=true
 maxDistance=60000
 soundClassObject=&Fort_Impacts_PSM
 volumeMultiplier=3.2
 attenuationOverrides {
  bAttenuate=false
  bEnableFocusInterpolation=true
  bEnableListenerFocus=true
  bEnableOcclusion=true
  focusAttackInterpSpeed=0.1
  focusAzimuth=15
  focusReleaseInterpSpeed=0.5
  nonFocusAzimuth=30
  nonFocusVolumeAttenuation=0
  occlusionInterpolationTime=2
  occlusionTraceChannel=ECC_GameTraceChannel11
  occlusionVolumeAttenuation=0
  stereoSpread=30000
 }
 concurrencyOverrides {maxCount=1, resolutionRule=StopOldest}
}
